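# Physics asset for the M_MED_UnbrokenTrash male player body cosmetic, set up for the motion of its vest flaps against the lower body.
# Capsules on the pelvis, spine_01 and thighs form the body volume, with tiny vest base capsules linked to spine_01 by angularly free constraints.
# Limited-swing joints hang single-segment back and side flaps and two-segment left and right vest chains from those bases, with twist mostly at zero.
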PhysicsAsset /BRCosmetics/Characters/Player/Male/Medium/Bodies/M_MED_UnbrokenTrash/Meshes/M_MED_UnbrokenTrash_Physics
bodies {
 base_dyn_vest_r [{shapeType=Capsule, length=0.505, radius=0.505, shapeName=base_dyn_vest_r_capsule, rotation=<rot 0 -90 0>}]
 base_vestBk_c [{shapeType=Capsule, length=0.505, radius=0.505, shapeName=dyn_vestBk_c_MSTR_capsule, rotation=<rot 0 -90 0>}]
 base_vestBk_l [{shapeType=Capsule, length=0.505, radius=0.505, shapeName=base_vestBk_l_capsule, rotation=<rot 0 -90 0>}]
 base_vestBk_r [{shapeType=Capsule, length=0.505, radius=0.505, shapeName=base_vestBk_r_capsule, rotation=<rot 0 -90 0>}]
 base_vestSide_l [{shapeType=Capsule, length=0.505, radius=0.505, shapeName=base_vestSide_l_capsule, rotation=<rot 0 -90 0>}]
 base_vestSide_r [{shapeType=Capsule, length=0.505, radius=0.505, shapeName=base_vestSide_r_capsule, rotation=<rot 0 -90 0>}]
 base_vest_l [{shapeType=Capsule, length=0.505, radius=0.505, shapeName=base_vest_l_capsule, rotation=<rot 0 -90 0>}]
 dyn_vestBk_c_01 [{shapeType=Capsule, length=12.5, radius=2.5, shapeName=dyn_vestBk_c_01_capsule, center=<5.8777 -2 0>, rotation=<rot 90 0 0>}]
 dyn_vestBk_l_01 [{shapeType=Capsule, length=12.5, radius=2.5, shapeName=dyn_vestBk_l_01_capsule, center=<5.8777 -3.7 0>, rotation=<rot 90 0 0>}]
 dyn_vestBk_r_01 [{shapeType=Capsule, length=12.5, radius=2.5, shapeName=dyn_vestBk_r_01_capsule, center=<5.8777 -3.7 0>, rotation=<rot 90 0 0>}]
 dyn_vestSide_l_01 [{shapeType=Capsule, length=15.5, radius=3, shapeName=dyn_vestSide_l_01_capsule, center=<5.8777 -3.7 0>, rotation=<rot 90 0 0>}]
 dyn_vestSide_r_01 [{shapeType=Capsule, length=12.5, radius=2.5, shapeName=dyn_vestSide_r_01_capsule, center=<5.8777 -4 0>, rotation=<rot 90 0 0>}]
 dyn_vest_l_01 [{shapeType=Capsule, length=10.5, radius=3, shapeName=dyn_vest_l_01_capsule, center=<4.8385 -6 0>, rotation=<rot 90 0 0>}]
 dyn_vest_l_02 [{shapeType=Capsule, length=18, radius=9.5, shapeName=dyn_vest_l_02_capsule, center=<5 -11 -7.4>, rotation=<rot 0.0001 -0.0001 -8.2532>}]
 dyn_vest_r_01 [{shapeType=Capsule, length=11, radius=2.5, shapeName=dyn_vest_r_01_capsule, center=<4.8385 -4 0>, rotation=<rot 90 0 0>}]
 dyn_vest_r_02 [{shapeType=Capsule, length=25, radius=8, shapeName=dyn_vest_r_02_capsule, center=<4.8385 -6 6>}]
 pelvis [
  {shapeType=Capsule, length=11.5, radius=14, shapeName=pelvis_capsule}
  {shapeType=Capsule, length=14.5, radius=14, shapeName=pelvis_capsule, center=<-6.3494 -2.4115 0>}
 ]
 spine_01 [{shapeType=Capsule, length=7, radius=12.5, shapeName=spine_01_capsule, rotation=<rot 0 -90 0>}]
 thigh_l [{shapeType=Capsule, length=47, radius=12.5, shapeName=thigh_l_capsule, center=<-23.8981 0 0>, rotation=<rot 90 0 0>}]
 thigh_r [{shapeType=Capsule, length=47, radius=13, shapeName=thigh_l_capsule, center=<23.8981 0 0>, rotation=<rot 90 180 0>}]
}
constraints [
 {
  bone1Name=spine_01
  bone2Name=pelvis
  swing1LimitDegrees=45
  swing1Motion=Free
  swing2LimitDegrees=45
  swing2Motion=Free
  twistLimitDegrees=45
  twistMotion=Free
 }
 {
  bone1Name=base_vestBk_c
  bone2Name=spine_01
  swing1LimitDegrees=45
  swing1Motion=Free
  swing2LimitDegrees=45
  swing2Motion=Free
  twistLimitDegrees=45
  twistMotion=Free
 }
 {bone1Name=dyn_vestBk_c_01, bone2Name=base_vestBk_c, swing1LimitDegrees=20, swing2LimitDegrees=20, twistLimitDegrees=0}
 {
  bone1Name=base_vestBk_l
  bone2Name=spine_01
  swing1LimitDegrees=45
  swing1Motion=Free
  swing2LimitDegrees=45
  swing2Motion=Free
  twistLimitDegrees=45
  twistMotion=Free
 }
 {bone1Name=dyn_vestBk_l_01, bone2Name=base_vestBk_l, swing1LimitDegrees=20, swing2LimitDegrees=20, twistLimitDegrees=0}
 {
  bone1Name=base_vestBk_r
  bone2Name=spine_01
  swing1LimitDegrees=45
  swing1Motion=Free
  swing2LimitDegrees=45
  swing2Motion=Free
  twistLimitDegrees=45
  twistMotion=Free
 }
 {bone1Name=dyn_vestBk_r_01, bone2Name=base_vestBk_r, swing1LimitDegrees=20, swing2LimitDegrees=20, twistLimitDegrees=0}
 {
  bone1Name=base_vestSide_l
  bone2Name=spine_01
  swing1LimitDegrees=45
  swing1Motion=Free
  swing2LimitDegrees=45
  swing2Motion=Free
  twistLimitDegrees=45
  twistMotion=Free
 }
 {bone1Name=dyn_vestSide_l_01, bone2Name=base_vestSide_l, swing1LimitDegrees=20, swing2LimitDegrees=20, twistLimitDegrees=0}
 {
  bone1Name=base_vestSide_r
  bone2Name=spine_01
  swing1LimitDegrees=45
  swing1Motion=Free
  swing2LimitDegrees=45
  swing2Motion=Free
  twistLimitDegrees=45
  twistMotion=Free
 }
 {bone1Name=dyn_vestSide_r_01, bone2Name=base_vestSide_r, swing1LimitDegrees=20, swing2LimitDegrees=20, twistLimitDegrees=0}
 {
  bone1Name=base_dyn_vest_r
  bone2Name=spine_01
  swing1LimitDegrees=45
  swing1Motion=Free
  swing2LimitDegrees=45
  swing2Motion=Free
  twistLimitDegrees=45
  twistMotion=Free
 }
 {bone1Name=dyn_vest_r_01, bone2Name=base_dyn_vest_r, swing1LimitDegrees=55, swing2LimitDegrees=15, twistLimitDegrees=0}
 {bone1Name=dyn_vest_r_02, bone2Name=dyn_vest_r_01, swing1LimitDegrees=70, swing2LimitDegrees=20, twistLimitDegrees=0}
 {
  bone1Name=base_vest_l
  bone2Name=spine_01
  swing1LimitDegrees=45
  swing1Motion=Free
  swing2LimitDegrees=45
  swing2Motion=Free
  twistLimitDegrees=45
  twistMotion=Free
 }
 {bone1Name=dyn_vest_l_01, bone2Name=base_vest_l, swing1LimitDegrees=55, swing2LimitDegrees=20, twistLimitDegrees=0}
 {bone1Name=dyn_vest_l_02, bone2Name=dyn_vest_l_01, swing1LimitDegrees=55, swing2LimitDegrees=20, twistLimitDegrees=30}
 {
  bone1Name=thigh_l
  bone2Name=pelvis
  swing1LimitDegrees=45
  swing1Motion=Free
  swing2LimitDegrees=45
  swing2Motion=Free
  twistLimitDegrees=45
  twistMotion=Free
 }
 {
  bone1Name=thigh_r
  bone2Name=pelvis
  swing1LimitDegrees=45
  swing1Motion=Free
  swing2LimitDegrees=45
  swing2Motion=Free
  twistLimitDegrees=45
  twistMotion=Free
 }
]
