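# Post-process style AnimBlueprint for the M_MED_UnbrokenTrash player body, adding secondary motion to clothing and gear.
# Copies the attached parent's pose into a Control Rig fed by movement flags, then chains weighted rivet constraints, AnimDynamics, RigidBody and corrective bone offsets through cached poses.
# The vest-back offsets cache (PreSim2) never reaches the output.
# The event graph updates those flags, gravity and sim-space settings, and hides tail meshes under capes.
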AnimBlueprint /BRCosmetics/Characters/Player/Male/Medium/Bodies/M_MED_UnbrokenTrash/Meshes/M_MED_UnbrokenTrash_AnimBP
target_skeleton=&/M_MED_UnbrokenTrash_Skeleton
graphs [
 AnimGraph {nodes [
   Output Pose(Result=Use cached pose 'CR_Cache_Dynamics'()) {layerGroup=None}
   n1 = Control Rig!(
    Source=Copy Pose From Mesh() {bUseAttachedParent=true}
    Alpha
    Pub_IsCrouching:double=Get Is Crouching() -> (Is Crouching:double)
    Pub_KneeSliding:double=Get IsKneeSliding() -> (IsKneeSliding:double)
    Pub_IsJumping:double=Get IsJumping() -> (IsJumping:double)
    Pub_IsCrouchingSprinting:double=Get IsCrouchingSpriting() -> (IsCrouchingSpriting:double)
    Pub_PouchesSide:double=Get IsPouchesSide() -> (IsPouchesSide:double)
    Pub_HangGliding:double=Get Is HangGliding() -> (Is HangGliding:double)
   ) -> (Pose) {
    alphaBoolBlend {blendOption=Linear}
    controlRigAssetReference {blueprintRigClass=&/M_MED_UnbrokenTrash_CtrlRig.M_MED_UnbrokenTrash_CtrlRig_C}
   }
   Save cached pose 'CR_Cache_Input'(Pose=@n1) {cacheName=CR_Cache_Input}
   n2 = "Constraint - base_weapon_r (PPPPPPPP)"!(
    ComponentPose=Local To Component(LocalPose=Use cached pose 'CR_Cache_Input'())
    Alpha=1
    ConstraintWeights_0=0.7687
    ConstraintWeights_1:float=0.0828
    ConstraintWeights_2:float=0.0798
    ConstraintWeights_3:float=0.0473
    ConstraintWeights_4:float=0.0101
    ConstraintWeights_5:float=0.0072
    ConstraintWeights_6:float=0.0031
    ConstraintWeights_7:float=0.0009
   ) -> (Pose) {
    alphaBoolBlend {blendOption=Linear}
    boneToModify {boneName=base_weapon_r}
    constraintSetup [
     {targetBone {boneName=base_weapon_r}}
     {targetBone {boneName=base_pouch_Thigh_r}}
     {targetBone {boneName=thigh_twist_01_r}}
     {targetBone {boneName=pelvis}}
     {targetBone {boneName=thigh_r}}
     {targetBone {boneName=deform_groin_r}}
     {targetBone {boneName=deform_glute_r}}
     {targetBone {boneName=deform_glute_l}}
    ]
    constraintWeights [1,1,1,1,1,1,1,1]
   }
   n3 = "Constraint - base_shoulderBuckle_l (PPPPPPPP)"!(
    ComponentPose=@n2
    Alpha=1
    ConstraintWeights_0=0.1915
    ConstraintWeights_1:float=0.1842
    ConstraintWeights_2:float=0.1801
    ConstraintWeights_3:float=0.1648
    ConstraintWeights_4:float=0.1194
    ConstraintWeights_5:float=0.1178
    ConstraintWeights_6:float=0.0242
    ConstraintWeights_7:float=0.0182
   ) -> (Pose) {
    alphaBoolBlend {blendOption=Linear}
    boneToModify {boneName=base_shoulderBuckle_l}
    constraintSetup [
     {targetBone {boneName=upperarm_twist_01_l}}
     {targetBone {boneName=clavicle_l}}
     {targetBone {boneName=deform_pec_l}}
     {targetBone {boneName=deform_upperarm_fr_l}}
     {targetBone {boneName=spine_04}}
     {targetBone {boneName=spine_03}}
     {targetBone {boneName=spine_02}}
     {targetBone {boneName=upperarm_twist_02_l}}
    ]
    constraintWeights [1,1,1,1,1,1,1,1]
   }
   n4 = "Constraint - base_pouch_Thigh_r (PPPPPPPP)"!(
    ComponentPose=@n3
    Alpha=1
    ConstraintWeights_0=0.6426
    ConstraintWeights_1:float=0.1787
    ConstraintWeights_2:float=0.0915
    ConstraintWeights_3:float=0.0601
    ConstraintWeights_4:float=0.0102
    ConstraintWeights_5:float=0.0068
    ConstraintWeights_6:float=0.0059
    ConstraintWeights_7:float=0.0043
   ) -> (Pose) {
    alphaBoolBlend {blendOption=Linear}
    boneToModify {boneName=base_pouch_Thigh_r}
    constraintSetup [
     {targetBone {boneName=base_pouch_Thigh_r}}
     {targetBone {boneName=base_weapon_r}}
     {targetBone {boneName=pelvis}}
     {targetBone {boneName=thigh_twist_01_r}}
     {targetBone {boneName=deform_glute_r}}
     {targetBone {boneName=thigh_r}}
     {targetBone {boneName=deform_glute_l}}
     {targetBone {boneName=thigh_twist_01_l}}
    ]
    constraintWeights [1,1,1,1,1,1,1,1]
   }
   Save cached pose 'CR_Cache_Rivet'(Pose=Component To Local(ComponentPose=@n4)) {cacheName=CR_Cache_Rivet}
   n5 = "Constraint - base_collar_a_l (P)"(ComponentPose=Local To Component(LocalPose=Use cached pose 'CR_Cache_Rivet'()), Alpha=0.8, ConstraintWeights_0=1) {
    alphaBoolBlend {blendOption=Linear}
    boneToModify {boneName=base_collar_a_l}
    constraintSetup [{targetBone {boneName=clavicle_l}}]
    constraintWeights [1]
   }
   n6 = "Constraint - base_collar_a_r (P)"(ComponentPose=@n5, Alpha=0.8, ConstraintWeights_0=1) {
    alphaBoolBlend {blendOption=Linear}
    boneToModify {boneName=base_collar_a_r}
    constraintSetup [{targetBone {boneName=clavicle_r}}]
    constraintWeights [1]
   }
   n7 = Anim Dynamics - Bone: dyn_collar_a_l_01!(SimulationSpace:AnimPhysSimSpaceType Enum=BoneRelative, ComponentPose=@n6, Alpha=1, AngularSpringConstant:float=6000, ExternalForce="0,0,0") -> (Pose) {
    angularDampingOverride=0.99
    angularSpringConstant=6000
    bAngularSpring=true
    bOverrideAngularDamping=true
    bUseGravityOverride=true
    gravityScale=0.35
    simulationSpace=BoneRelative
    alphaBoolBlend {blendOption=Linear}
    boundBone {boneName=dyn_collar_a_l_01}
    componentAppliedLinearAccClamp=<10000 10000 10000>
    componentLinearAccScale=<1 1 1>
    componentLinearVelScale=<1 1 1>
    gravityOverride {z=-980}
    physicsBodyDefinitions [{
      boundBone {boneName=dyn_collar_a_l_01}
      boxExtents=<20 30 30>
      constraintSetup {angularLimitsMax=<0 5 0>, angularLimitsMin=<0 -5 0>, angularTarget=<1 0 0>}
      localJointOffset=<4 0 0>
     }]
    relativeSpaceBone {boneName=root}
   }
   n8 = Anim Dynamics - Bone: dyn_collar_a_r_01!(SimulationSpace:AnimPhysSimSpaceType Enum=BoneRelative, ComponentPose=@n7, Alpha=1, AngularSpringConstant:float=6000, ExternalForce="0,0,0") -> (Pose) {
    angularDampingOverride=0.99
    angularSpringConstant=6000
    bAngularSpring=true
    bOverrideAngularDamping=true
    bUseGravityOverride=true
    gravityScale=0.35
    simulationSpace=BoneRelative
    alphaBoolBlend {blendOption=Linear}
    boundBone {boneName=dyn_collar_a_r_01}
    componentAppliedLinearAccClamp=<10000 10000 10000>
    componentLinearAccScale=<1 1 1>
    componentLinearVelScale=<1 1 1>
    gravityOverride {z=-980}
    physicsBodyDefinitions [{
      boundBone {boneName=dyn_collar_a_r_01}
      boxExtents=<20 30 30>
      constraintSetup {angularLimitsMax=<0 5 0>, angularLimitsMin=<0 -5 0>, angularTarget=<1 0 0>}
      localJointOffset=<4 0 0>
     }]
    relativeSpaceBone {boneName=root}
   }
   Save cached pose 'CR_Cache_PreSim'(Pose=Component To Local(ComponentPose=@n8)) {cacheName=CR_Cache_PreSim}
   n9 = "Transform (Modify) Bone - Bone: base_vestBk_l"!(
    ComponentPose=Local To Component(LocalPose=Use cached pose 'CR_Cache_PreSim'())
    bAlphaBoolEnabled:bool=True
    Translation="0,-4,0"
    Rotation="0,-30,0"
    Scale="1,1,1"
   ) -> (Pose) {
    alphaInputType=Bool
    rotationMode=BMM_Additive
    rotationSpace=BCS_BoneSpace
    scaleMode=BMM_Additive
    scaleSpace=BCS_BoneSpace
    translationMode=BMM_Additive
    translationSpace=BCS_BoneSpace
    alphaBoolBlend {blendInTime=0.2, blendOption=Linear, blendOutTime=0.2}
    boneToModify {boneName=base_vestBk_l}
    rotation {yaw=-30}
    translation {y=-4}
   }
   n10 = "Transform (Modify) Bone - Bone: base_vestBk_c"!(ComponentPose=@n9, bAlphaBoolEnabled:bool=True, Translation="-3,-2,0", Rotation="0,-30,0", Scale="0.8,1,1") -> (Pose) {
    alphaInputType=Bool
    rotationMode=BMM_Additive
    rotationSpace=BCS_BoneSpace
    scaleMode=BMM_Additive
    scaleSpace=BCS_BoneSpace
    translationMode=BMM_Additive
    translationSpace=BCS_BoneSpace
    alphaBoolBlend {blendInTime=0.2, blendOption=Linear, blendOutTime=0.2}
    boneToModify {boneName=base_vestBk_c}
    rotation {yaw=-30}
    scale {x=0.8}
    translation=<-3 -2>
   }
   n11 = "Transform (Modify) Bone - Bone: base_vestBk_r"!(ComponentPose=@n10, bAlphaBoolEnabled:bool=True, Translation="0,-2,0", Rotation="0,0,0", Scale="1,1,1") -> (Pose) {
    alphaInputType=Bool
    rotationMode=BMM_Additive
    rotationSpace=BCS_BoneSpace
    scaleMode=BMM_Additive
    scaleSpace=BCS_BoneSpace
    translationMode=BMM_Additive
    translationSpace=BCS_BoneSpace
    alphaBoolBlend {blendInTime=0.2, blendOption=Linear, blendOutTime=0.2}
    boneToModify {boneName=base_vestBk_r}
    translation {y=-2}
   }
   Save cached pose 'CR_Cache_PreSim2'(Pose=Component To Local(ComponentPose=@n11)) {cacheName=CR_Cache_PreSim2}
   n12 = RigidBody!(
    ComponentPose=Local To Component(LocalPose=Use cached pose 'CR_Cache_PreSim'())
    Alpha
    OverrideWorldGravity:Vector="0,0,-980"
    ExternalForce
   ) -> (Pose) {
    bClampLinearTranslationLimitToRefPose=true
    bOverrideWorldGravity=true
    overridePhysicsAsset=&/M_MED_UnbrokenTrash_Physics
    simulationSpace=BaseBoneSpace
    alphaBoolBlend {blendOption=Linear}
    baseBoneRef {boneName=pelvis}
    componentAppliedLinearAccClamp=<6000 6000 6000>
    componentLinearAccScale=<0.1 0.3 0.3>
    componentLinearVelScale=<0.1 0.3 0.3>
    overrideWorldGravity {z=-980}
   }
   n13 = Anim Dynamics - Bone: dyn_weapon_r!(
    SimulationSpace:AnimPhysSimSpaceType Enum
    ComponentPose=@n12
    Alpha=1
    GravityOverride:Vector="0,0,0"
    AngularSpringConstant:float=3000
    ExternalForce="0,0,0"
   ) -> (Pose) {
    angularDampingOverride=1
    angularSpringConstant=3000
    bAngularSpring=true
    bOverrideAngularDamping=true
    bUseGravityOverride=true
    gravityScale=0.35
    simulationSpace=RootRelative
    alphaBoolBlend {blendOption=Linear}
    boundBone {boneName=dyn_weapon_r}
    componentAppliedLinearAccClamp=<10000 10000 10000>
    componentLinearAccScale=<1 1 1>
    componentLinearVelScale=<1 1 1>
    physicsBodyDefinitions [{
      boundBone {boneName=dyn_weapon_r}
      boxExtents=<25 40 40>
      constraintSetup {angularLimitsMax=<15 0 20>, angularLimitsMin=<-15 -10 -5>, angularTarget=<1 0 0>}
      localJointOffset=<4 0 0>
     }]
    relativeSpaceBone {boneName=root}
   }
   n14 = Anim Dynamics - Bone: dyn_loop_Fr_l_01!(
    SimulationSpace:AnimPhysSimSpaceType Enum
    ComponentPose=@n13
    Alpha=1
    GravityOverride:Vector="0,0,-980"
    AngularSpringConstant:float=20000
    ExternalForce="0,0,0"
   ) -> (Pose) {
    angularDampingOverride=1
    angularSpringConstant=20000
    bAngularSpring=true
    bOverrideAngularDamping=true
    bUseGravityOverride=true
    gravityScale=0.35
    simulationSpace=RootRelative
    alphaBoolBlend {blendOption=Linear}
    boundBone {boneName=dyn_loop_Fr_l_01}
    componentAppliedLinearAccClamp=<10000 10000 10000>
    componentLinearAccScale=<1 1 1>
    componentLinearVelScale=<1 1 1>
    gravityOverride {z=-980}
    physicsBodyDefinitions [{
      boundBone {boneName=dyn_loop_Fr_l_01}
      boxExtents=<20 30 30>
      constraintSetup {angularLimitsMax=<0 0 2>, angularLimitsMin=<0 0 -30>, angularTarget=<1 0 0>}
      localJointOffset=<3 0 0>
     }]
   }
   n15 = Anim Dynamics - Bone: dyn_loop_Bk_l_01!(
    SimulationSpace:AnimPhysSimSpaceType Enum
    ComponentPose=@n14
    Alpha=1
    GravityOverride:Vector="0,0,-980"
    AngularSpringConstant:float=20000
    ExternalForce="0,0,0"
   ) -> (Pose) {
    angularDampingOverride=1
    angularSpringConstant=20000
    bAngularSpring=true
    bOverrideAngularDamping=true
    bUseGravityOverride=true
    gravityScale=0.35
    simulationSpace=RootRelative
    alphaBoolBlend {blendOption=Linear}
    boundBone {boneName=dyn_loop_Bk_l_01}
    componentAppliedLinearAccClamp=<10000 10000 10000>
    componentLinearAccScale=<1 1 1>
    componentLinearVelScale=<1 1 1>
    gravityOverride {z=-980}
    physicsBodyDefinitions [{
      boundBone {boneName=dyn_loop_Bk_l_01}
      boxExtents=<20 30 30>
      constraintSetup {angularLimitsMax=<0 0 0>, angularLimitsMin=<0 0 -30>, angularTarget=<1 0 0>}
      localJointOffset=<3 0 0>
     }]
   }
   n16 = Anim Dynamics - Bone: dyn_shoulderBuckle_l_01!(
    SimulationSpace:AnimPhysSimSpaceType Enum=World
    ComponentPose=@n15
    Alpha=1
    GravityOverride:Vector="0,0,-980"
    AngularSpringConstant:float=20000
    ExternalForce="0,0,0"
   ) -> (Pose) {
    angularDampingOverride=1
    angularSpringConstant=20000
    bAngularSpring=true
    bOverrideAngularDamping=true
    bUseGravityOverride=true
    gravityScale=0.35
    simulationSpace=World
    alphaBoolBlend {blendOption=Linear}
    boundBone {boneName=dyn_shoulderBuckle_l_01}
    componentAppliedLinearAccClamp=<10000 10000 10000>
    componentLinearAccScale=<1 1 1>
    componentLinearVelScale=<1 1 1>
    gravityOverride {z=-980}
    physicsBodyDefinitions [{
      boundBone {boneName=dyn_shoulderBuckle_l_01}
      boxExtents=<20 30 30>
      constraintSetup {angularLimitsMax=<0 80 0>, angularLimitsMin=<0 -80 0>, angularTarget=<1 0 0>}
      localJointOffset=<4 0 0>
     }]
   }
   n17 = Anim Dynamics - Bone: dyn_spurs_l!(
    SimulationSpace:AnimPhysSimSpaceType Enum=Component
    ComponentPose=@n16
    Alpha=1
    GravityOverride:Vector="0,0,-980"
    AngularSpringConstant:float=20000
    ExternalForce="0,0,0"
   ) -> (Pose) {
    angularDampingOverride=1
    angularSpringConstant=20000
    bAngularSpring=true
    bOverrideAngularDamping=true
    bUseGravityOverride=true
    gravityScale=0.35
    alphaBoolBlend {blendOption=Linear}
    boundBone {boneName=dyn_spurs_l}
    componentAppliedLinearAccClamp=<10000 10000 10000>
    componentLinearAccScale=<1 1 1>
    componentLinearVelScale=<1 1 1>
    gravityOverride {z=-980}
    physicsBodyDefinitions [{
      boundBone {boneName=dyn_spurs_l}
      boxExtents=<20 30 30>
      constraintSetup {angularLimitsMax=<0 0 80>, angularLimitsMin=<0 0 -80>, angularTarget=<1 0 0>}
      localJointOffset=<4 0 0>
     }]
   }
   n18 = Anim Dynamics - Bone: dyn_spurs_r!(
    SimulationSpace:AnimPhysSimSpaceType Enum=Component
    ComponentPose=@n17
    Alpha=1
    GravityOverride:Vector="0,0,-980"
    AngularSpringConstant:float=20000
    ExternalForce="0,0,0"
   ) -> (Pose) {
    angularDampingOverride=1
    angularSpringConstant=20000
    bAngularSpring=true
    bOverrideAngularDamping=true
    bUseGravityOverride=true
    gravityScale=0.35
    alphaBoolBlend {blendOption=Linear}
    boundBone {boneName=dyn_spurs_r}
    componentAppliedLinearAccClamp=<10000 10000 10000>
    componentLinearAccScale=<1 1 1>
    componentLinearVelScale=<1 1 1>
    gravityOverride {z=-980}
    physicsBodyDefinitions [{
      boundBone {boneName=dyn_spurs_r}
      boxExtents=<20 30 30>
      constraintSetup {angularLimitsMax=<0 0 80>, angularLimitsMin=<0 0 -80>, angularTarget=<1 0 0>}
      localJointOffset=<4 0 0>
     }]
   }
   n19 = "Transform (Modify) Bone - Bone: dyn_loop_Fr_l_01"!(ComponentPose=@n18, bAlphaBoolEnabled:bool=True, Translation="0,0,0", Rotation="6,0,-2", Scale="1,1,1") -> (Pose) {
    alphaInputType=Bool
    rotationMode=BMM_Additive
    translationMode=BMM_Additive
    alphaBoolBlend {blendOption=Linear}
    alphaScaleBiasClamp {bInterpResult=true, interpSpeedDecreasing=5, interpSpeedIncreasing=5}
    boneToModify {boneName=dyn_loop_Fr_l_01}
    rotation {pitch=6, roll=-2}
   }
   n20 = "Constraint - base_zipper_l (PPPPPPP)"!(
    ComponentPose=@n19
    Alpha=1
    ConstraintWeights_0=0.4548
    ConstraintWeights_1:float=0.3788
    ConstraintWeights_2:float=0.0785
    ConstraintWeights_3:float=0.0333
    ConstraintWeights_4:float=0.0224
    ConstraintWeights_5:float=0.0164
    ConstraintWeights_6:float=0.0159
   ) -> (Pose) {
    alphaBoolBlend {blendOption=Linear}
    boneToModify {boneName=base_zipper_l}
    constraintSetup [
     {targetBone {boneName=dyn_vest_l_02}}
     {targetBone {boneName=dyn_vest_l_01}}
     {targetBone {boneName=dyn_vestSide_l_01}}
     {targetBone {boneName=pelvis}}
     {targetBone {boneName=spine_02}}
     {targetBone {boneName=spine_01}}
     {targetBone {boneName=spine_03}}
    ]
    constraintWeights [1,1,1,1,1,1,1]
   }
   n21 = "Transform (Modify) Bone - Bone: lowerarm_l"!(ComponentPose=@n20, bAlphaBoolEnabled:bool=True, Translation="0,0,0", Rotation="0,0,0", Scale="0.8,0.8,0.8") -> (Pose) {
    alphaInputType=Bool
    rotationMode=BMM_Additive
    scaleMode=BMM_Additive
    scaleSpace=BCS_BoneSpace
    translationMode=BMM_Additive
    alphaBoolBlend {blendOption=Linear}
    boneToModify {boneName=lowerarm_l}
    scale=<0.8 0.8 0.8>
   }
   n22 = "Transform (Modify) Bone - Bone: dyn_sleeve_01_l"!(ComponentPose=@n21, bAlphaBoolEnabled:bool=True, Translation="0,0,0", Rotation="0,-10,0", Scale="1,1,1") -> (Pose) {
    alphaInputType=Bool
    rotationMode=BMM_Additive
    rotationSpace=BCS_BoneSpace
    scaleMode=BMM_Additive
    scaleSpace=BCS_BoneSpace
    translationMode=BMM_Additive
    alphaBoolBlend {blendOption=Linear}
    boneToModify {boneName=dyn_sleeve_01_l}
    rotation {yaw=-10}
   }
   n23 = "Transform (Modify) Bone - Bone: dyn_sleeve_02_l"!(ComponentPose=@n22, bAlphaBoolEnabled:bool=True, Translation="0,0,0", Rotation="0,-10,-25", Scale="1,1,1") -> (Pose) {
    alphaInputType=Bool
    rotationMode=BMM_Additive
    rotationSpace=BCS_BoneSpace
    scaleMode=BMM_Additive
    scaleSpace=BCS_BoneSpace
    translationMode=BMM_Additive
    alphaBoolBlend {blendOption=Linear}
    boneToModify {boneName=dyn_sleeve_02_l}
    rotation {roll=-25, yaw=-10}
   }
   n24 = "Transform (Modify) Bone - Bone: dyn_sleeve_01_r"!(ComponentPose=@n23, bAlphaBoolEnabled:bool=True, Translation="3,0,0", Rotation="0,0,0", Scale="1,1,1") -> (Pose) {
    alphaInputType=Bool
    rotationMode=BMM_Additive
    scaleMode=BMM_Additive
    scaleSpace=BCS_BoneSpace
    translationMode=BMM_Additive
    alphaBoolBlend {blendOption=Linear}
    boneToModify {boneName=dyn_sleeve_01_r}
    translation {x=3}
   }
   Save cached pose 'CR_Cache_Dynamics'(Pose=Component To Local(ComponentPose=@n24)) {cacheName=CR_Cache_Dynamics}
  ]}
 EventGraph {nodes [
   Event BlueprintUpdateAnimation()
   > Set Delta Time X(Delta Time X=0)
   > n1 = Sequence!(execute) -> (then_0, then_1, then_2:Exec, then_3:Exec)
   > n2 = Gravity Override State Function DEPRECATED(
    CharacterPartAnimInstance=Self-Reference()
    GameTimeInSeconds=GetGameTimeInSeconds()
    GravityOverrideParams=Get Gravity Override Params()
    RBAN Params=Get RBAN Params()
   )
   > Set GravityOverride(GravityOverride=@n2)
   > Set Sim Space Settings(Sim Space Settings:Sim Space Settings Structure=Break RBANSetting Struct!(RBANSettingStruct:"RBANSetting Struct Structure (by ref)"=@n2.RBAN SettingsOut) -> (
     ComponentLinearAccScale_28_#1:Vector=~
     ComponentLinearVelScale_31_#2:Vector=~
     SimSpaceSettings_34_#3:Sim Space Settings Structure
    ) {is=title K2Node_BreakStruct}.SimSpaceSettings_34_#3) -> (Output_Get:Sim Space Settings Structure)
   > Set IsCrouchingSpriting(IsCrouchingSpriting:double="To Float (Boolean)"(InBool=Get bIsCrouchSprinting())) -> (Output_Get:double)
   n3 = PlayerMovement(execute=@n1.then_1, AnimInstance=Self-Reference(), XDotProductCoefficient=100, YDotProductCoefficient=100, ZDotProductCoefficient=100)
   n4 = BreakVector(InVec=MakeVector(X=@n3, Y=@n3.YMovement, Z=@n3.ZMovement))
   > n5 = Set Left(Left:bool="float < float"(A=@n4, B=-50)) -> (Output_Get:bool)
   > n6 = Set Right(Right:bool="float > float"(A=@n4, B=50)) -> (Output_Get:bool)
   > Set Forward(Forward:bool="float > float"(A=@n4.Y, B=50)) -> (Output_Get:bool)
   > n7 = Set Backward(Backward:bool="float < float"(A=@n4.Y, B=-50)) -> (Output_Get:bool)
   > n8 = Branch(Condition=AND Boolean(A=OR Boolean!(A=@n5, B=@n6, C:bool=@n7) -> (ReturnValue), B=Get bIsCrouching()))
   > Set IsPouchesSide(IsPouchesSide:double=1) -> (Output_Get:double)
   Set IsPouchesSide(execute=@n8.else, IsPouchesSide=0)
   n9 = Branch(execute=@n1.then_2, Condition=Get bIsInFrontEnd())
   > Set SimlationSpace(SimlationSpace:AnimPhysSimSpaceType Enum=World) -> (Output_Get:AnimPhysSimSpaceType Enum)
   > Set AlphaFE(AlphaFE:double=0.3) -> (Output_Get:double)
   Set SimlationSpace(execute=@n9.else)
   > Set AlphaFE(AlphaFE=1)
   Event OnBeginCrouching()
   > Set Is Crouching(Is Crouching:double=1) -> (Output_Get:double)
   Event OnParachuteOpened()
   > Set Is HangGliding(Is HangGliding:double=1) -> (Output_Get:double)
   Event OnParachuteClosed()
   > Set Is HangGliding(Is HangGliding=0)
   Event OnWeaponChanged()
   > Set Pistol01(Pistol01:double=0) -> (Output_Get:double)
   > Set Shotgun(Shotgun:double=0) -> (Output_Get:double)
   > n10 = Switch on EFortWeaponCoreAnimation!(execute, Selection:EFortWeaponCoreAnimation Enum=Get WeaponCoreAnim!() -> (WeaponCoreAnim)) -> (
    Melee:Exec
    Pistol:Exec
    Shotgun:Exec
    PaperBlueprint:Exec
    Rifle:Exec
    MeleeOneHand:Exec
    MachinePistol:Exec
    RocketLauncher:Exec
    GrenadeLauncher:Exec
    GoingCommando:Exec
    AssaultRifle:Exec
    TacticalShotgun:Exec
    SniperRifle:Exec
    TrapPlacement:Exec
    ShoulderLauncher:Exec
    AbilityDecoTool:Exec
    Crossbow:Exec
    C4:Exec
    RemoteControl:Exec
    DualWield:Exec
    AR_BullPup:Exec
    AR_ForwardGrip:Exec
    MedPackPaddles:Exec
    SMG_P90:Exec
    AR_DrumGun:Exec
    Consumable_Small:Exec
    Consumable_Large:Exec
    Balloon:Exec
    MountedTurret:Exec
    CreativeTool:Exec
    ExplosiveBow:Exec
    AshtonIndigo:Exec
    AshtonChicago:Exec
    MeleeDualWield:Exec
    Unarmed:Exec
   )
   Set Pistol01(execute=@n10.Pistol, Pistol01=1)
   Set Shotgun(execute=@n10.Shotgun, Shotgun=1)
   Event OnEndCrouching()
   > Set Is Crouching(Is Crouching=0)
   n11 = Event OnCharacterCustomizationLoaded()
   > Set TailMesh(TailMesh:Skeletal Mesh Component Object Reference=Get Array Item with EFort Custom Part Type(Array=@n11.CustomMeshParts, PartType=MiscOrTail)) -> (Output_Get:Skeletal Mesh Component Object Reference)
   > n12 = Sequence!(execute) -> (then_0, then_1, then_2)
   n13 = CheckForBackbling()
   > Set isWearingCape(isWearingCape=OR Boolean!(A=@n13.bIsBackblingACape, B=@n13.bIsBackblingLong) -> (ReturnValue))
   > Set characterHasTail(characterHasTail:bool=CheckForTail!(self:Custom Character Part Anim Instance Object Reference) -> (bHasTail:bool)) -> (Output_Get:bool)
   Branch(execute=@n12.then_1, Condition=Get characterHasTail() -> (characterHasTail:bool))
   > n14 = Branch(Condition=Get isWearingCape() -> (isWearingCape:bool))
   n15 = Get TailMesh() -> (TailMesh:Skeletal Mesh Component Object Reference)
   > SetVisibility!(execute, self:Scene Component Object Reference=@n15, bNewVisibility:bool=false, bPropagateToChildren:bool=false) -> (then)
   SetVisibility(execute=@n14.else, self=@n15, bNewVisibility=true, bPropagateToChildren=false)
   n16 = Branch(
    execute=@n12.then_2
    Condition=IsNullItem!(ItemObject:Object Reference=Get Array Item with EFort Custom Part Type(Array:Array of Object References=@n11.CustomParts, PartType=Gameplay) -> (Output:Object Reference)) -> (Is Null Item:bool)
   )
   > Set IsWearingGuantlet(IsWearingGuantlet:bool=false) -> (Output_Get:bool)
   Set IsWearingGuantlet(execute=@n16.else, IsWearingGuantlet=true)
   Event OnEndKneeSliding() {is=title K2Node_Event}
   > Set IsKneeSliding(IsKneeSliding:double=0) -> (Output_Get:double)
   Event OnBeginJumping()
   > Set IsJumping(IsJumping:double=1) -> (Output_Get:double)
   Event OnEndJumping()
   > Set IsJumping(IsJumping=0)
   Event OnBeginFalling()
   > Set IsJumping(IsJumping=1)
   Event OnEndFalling()
   > Set IsJumping(IsJumping=0)
   Event OnBeginKneeSliding() {is=title K2Node_Event}
   > Set IsKneeSliding(IsKneeSliding=1)
  ]}
]
variables [
 AlphaFE {type=number}
 Backward {type=boolean}
 BodyMesh {type=SkeletalMeshComponent}
 characterHasTail {type=boolean}
 Custom Parts {type="array<CustomCharacterPart>", default []}
 Delta Time X {type=number}
 Forward {type=boolean}
 Gravity Override Params {type=GravityOverrideParamsStruct}
 GravityOverride {type=Vector}
 headRotX {type=number}
 headRotYNeg {type=number}
 headRotZNeg {type=number}
 headTwistNeg {type=number}
 headTwistPos {type=number}
 Is Crouching {type=number}
 Is HangGliding {type=number}
 IsCrouchingSpriting {type=number}
 IsJumping {type=number}
 IsKneeSliding {type=number}
 IsPouchesSide {type=number}
 isWearingCape {type=boolean}
 IsWearingGuantlet {type=boolean}
 Left {type=boolean}
 Pistol01 {type=number}
 RBAN Params {
  type=RBANParamsStruct
  default {
   dBNO {simSpaceSettings {
     maxAngularAcceleration=1000
     maxAngularVelocity=3
     maxLinearAcceleration=300
     maxLinearVelocity=700
     velocityScaleZ=0
     externalLinearDragV=<0 0.5 0>
    }}
   dBNO_Carried {simSpaceSettings {
     maxAngularAcceleration=1000
     maxAngularVelocity=3
     maxLinearAcceleration=300
     maxLinearVelocity=700
     velocityScaleZ=0
     externalLinearDragV=<0 0.5 0>
    }}
   emoteOrMelee {simSpaceSettings {
     maxAngularAcceleration=1000
     maxAngularVelocity=3
     maxLinearAcceleration=300
     maxLinearVelocity=700
     velocityScaleZ=0
     externalLinearDragV=<0 0.5 0>
    }}
   front_End {simSpaceSettings {
     maxAngularAcceleration=1000
     maxAngularVelocity=3
     maxLinearAcceleration=300
     maxLinearVelocity=700
     velocityScaleZ=0
     externalLinearDragV=<0 0.5 0>
    }}
   onGround_Crouch_Moving {simSpaceSettings {
     maxAngularAcceleration=1000
     maxAngularVelocity=3
     maxLinearAcceleration=300
     maxLinearVelocity=700
     velocityScaleZ=0
     externalLinearDragV=<0 0.5 0>
    }}
   onGround_Falling {simSpaceSettings {
     maxAngularAcceleration=1000
     maxAngularVelocity=3
     maxLinearAcceleration=300
     maxLinearVelocity=700
     velocityScaleZ=0
     externalLinearDragV=<0 0.5 0>
    }}
   onGround_Moving {simSpaceSettings {
     maxAngularAcceleration=1000
     maxAngularVelocity=3
     maxLinearAcceleration=300
     maxLinearVelocity=700
     velocityScaleZ=0
     externalLinearDragV=<0 0.5 0>
    }}
   onGround_Standing {simSpaceSettings {
     maxAngularAcceleration=1000
     maxAngularVelocity=3
     maxLinearAcceleration=300
     maxLinearVelocity=700
     velocityScaleZ=0
     externalLinearDragV=<0 0.5 0>
    }}
   rail_Grinding {simSpaceSettings {
     maxAngularAcceleration=1000
     maxAngularVelocity=3
     maxLinearAcceleration=300
     maxLinearVelocity=700
     velocityScaleZ=0
     worldAlpha=0.1
     externalLinearDragV=<0 0.5 0>
    }}
   riding_Moving {simSpaceSettings {
     maxAngularAcceleration=1000
     maxAngularVelocity=3
     maxLinearAcceleration=300
     maxLinearVelocity=700
     velocityScaleZ=0
     externalLinearDragV=<0 0.5 0>
    }}
   riding_Moving_Targeting {simSpaceSettings {
     maxAngularAcceleration=1000
     maxAngularVelocity=3
     maxLinearAcceleration=300
     maxLinearVelocity=700
     velocityScaleZ=0
     externalLinearDragV=<0 0.5 0>
    }}
   riding_Standing {simSpaceSettings {
     maxAngularAcceleration=1000
     maxAngularVelocity=3
     maxLinearAcceleration=300
     maxLinearVelocity=700
     velocityScaleZ=0
     externalLinearDragV=<0 0.5 0>
    }}
   skydive_DivingDown {simSpaceSettings {
     maxAngularAcceleration=1000
     maxAngularVelocity=3
     maxLinearAcceleration=300
     maxLinearVelocity=700
     velocityScaleZ=0
     externalLinearDragV=<0 0.5 0>
    }}
   skydive_FloatingDown {
    componentLinearAccScale=<0 0.5 0>
    componentLinearVelScale=<0 0.5 0>
    simSpaceSettings {
     maxAngularAcceleration=1000
     maxAngularVelocity=3
     maxLinearAcceleration=300
     maxLinearVelocity=700
     velocityScaleZ=0
     externalLinearDragV=<0 0.5 0>
    }
   }
   skydive_GoingUp {simSpaceSettings {
     maxAngularAcceleration=1000
     maxAngularVelocity=3
     maxLinearAcceleration=300
     maxLinearVelocity=700
     velocityScaleZ=0
     externalLinearDragV=<0 0.5 0>
    }}
   skydive_Parachuting {
    componentLinearAccScale=<0 0.1 0.3>
    componentLinearVelScale=<0 0.1 0.3>
    simSpaceSettings {
     maxAngularAcceleration=1000
     maxAngularVelocity=3
     maxLinearAcceleration=300
     maxLinearVelocity=700
     velocityScaleZ=0
     externalLinearDragV=<0 0.5 0>
    }
   }
   skydive_Surfing {simSpaceSettings {
     maxAngularAcceleration=1000
     maxAngularVelocity=3
     maxLinearAcceleration=300
     maxLinearVelocity=700
     velocityScaleZ=0
     externalLinearDragV=<0 0.5 0>
    }}
   swimming {simSpaceSettings {
     maxAngularAcceleration=1000
     maxAngularVelocity=3
     maxLinearAcceleration=300
     maxLinearVelocity=700
     velocityScaleZ=0
     externalLinearDragV=<0 0.5 0>
    }}
   tactical_Sprint {simSpaceSettings {
     maxAngularAcceleration=1000
     maxAngularVelocity=3
     maxLinearAcceleration=300
     maxLinearVelocity=700
     velocityScaleZ=0
     externalLinearDragV=<0 0.5 0>
    }}
   ziplining {simSpaceSettings {
     maxAngularAcceleration=1000
     maxAngularVelocity=3
     maxLinearAcceleration=300
     maxLinearVelocity=700
     velocityScaleZ=0
     externalLinearDragV=<0 0.5 0>
    }}
  }
 }
 Right {type=boolean}
 Shotgun {type=number}
 Sim Space Settings {
  type=SimSpaceSettings
  default {
   dampingAlpha=1
   maxAngularAcceleration=10000
   maxAngularVelocity=10000
   maxLinearAcceleration=10000
   maxLinearVelocity=10000
   velocityScaleZ=1
   worldAlpha=0
   externalLinearDragV=<0 0 0>
  }
 }
 SimlationSpace {type=AnimPhysSimSpaceType}
 TailMesh {type=SkeletalMeshComponent}
]
